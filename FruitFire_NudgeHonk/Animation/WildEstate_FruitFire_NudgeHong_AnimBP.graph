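# Item-layer AnimBlueprint for the FruitFire NudgeHonk on the Fortnite player skeleton, animating its awareness states; only ItemFinalPoseOverride has logic.
# Its state machine moves between Unaware, alert, combat and death using inherited flags like bAlerted, bInCombat and bDead.
# Unaware, Combat and Death pass the cached incoming pose through, while alert states play float and transition clips.
# Combat entry blends sync-grouped summon clips full-body or from Spine_01 by bExitLand, ending early on firing.
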AnimBlueprint /FruitFire_NudgeHonk/Animation/WildEstate_FruitFire_NudgeHong_AnimBP
target_skeleton=&Fortnite_M_Avg_Player_Skeleton
graphs [
 AnimGraph {nodes [Output Pose()]}
 EventGraph {links [], nodes []}
 ItemAimOffset {nodes [
   Input Pose!() -> (Pose, AimOffsetAlpha:double, Yaw:double, Pitch:double) {name=InPose_AimOffset}
   Input Pose!() -> (Pose) {name=InPose_PreAimOffsets}
   Output Pose() {layerGroup=Item Override}
  ]}
 ItemFallAdditive {nodes [
   Input Pose() {name=InPoseItemFallAdditive}
   Output Pose() {layerGroup=Item Override}
  ]}
 ItemFallLandAdditive {nodes [
   Input Pose() {name=InPoseItemFallLandAdditive}
   Output Pose() {layerGroup=Item Override}
  ]}
 ItemFallLevel2Override {nodes [
   Input Pose() {name=InPose_FallLevel2}
   Output Pose() {layerGroup=Item Override}
  ]}
 ItemFinalPoseOverride {nodes [
   AnimGraphNode_StateMachine_0 = StateMachine() {is=title AnimGraphNode_StateMachine}
   Output Pose(Result=@AnimGraphNode_StateMachine_0) {layerGroup=Item Override}
   Save cached pose 'FinalPoseCache'(Pose=Input Pose() {name=InFinalPose}) {cacheName=FinalPoseCache}
  ]}
 StateMachine {
  entry=Unaware
  path=ItemFinalPoseOverride.AnimGraphNode_StateMachine_0.StateMachine
  ids=id
  nodes [
   ToCombat() {is=title AnimStateAliasNode, also {aliases [AlertedCycle, CombatToAlerted, intoAlert, Unaware]}}
   Alias() {is=title AnimStateAliasNode, bGlobalAlias=true}
   AnimStateNode_0 = Death() {is=title AnimStateNode}
   AnimStateNode_10 = TransIntoCombat() {is=title AnimStateNode}
   AnimStateNode_11 = CombatToAlerted() {is=title AnimStateNode}
   AnimStateNode_5 = Combat() {is=title AnimStateNode}
   AnimStateNode_6 = Unaware() {is=title AnimStateNode}
   AnimStateNode_7 = intoAlert() {is=title AnimStateNode}
   AnimStateNode_8 = AlertedCycle() {is=title AnimStateNode}
   AnimStateTransitionNode_0 = AnimStateTransitionNode() {bAllowInertializationForSelfTransitions=true}
   AnimStateTransitionNode_1 = AnimStateTransitionNode() {bAllowInertializationForSelfTransitions=true}
   AnimStateTransitionNode_10 = AnimStateTransitionNode() {bAllowInertializationForSelfTransitions=true}
   AnimStateTransitionNode_11 = AnimStateTransitionNode() {bAllowInertializationForSelfTransitions=true, bAutomaticRuleBasedOnSequencePlayerInState=true}
   AnimStateTransitionNode_13 = AnimStateTransitionNode() {bAllowInertializationForSelfTransitions=true}
   AnimStateTransitionNode_14 = AnimStateTransitionNode() {bAllowInertializationForSelfTransitions=true, crossfadeDuration=0.3}
   AnimStateTransitionNode_15 = AnimStateTransitionNode() {bAllowInertializationForSelfTransitions=true, bAutomaticRuleBasedOnSequencePlayerInState=true}
   AnimStateTransitionNode_2 = AnimStateTransitionNode() {bAllowInertializationForSelfTransitions=true, bAutomaticRuleBasedOnSequencePlayerInState=true}
   AnimStateTransitionNode_21 = AnimStateTransitionNode() {bAllowInertializationForSelfTransitions=true, priorityOrder=2}
   AnimStateTransitionNode_22 = AnimStateTransitionNode() {bAllowInertializationForSelfTransitions=true}
  ]
  links [
   {
    dst=TransIntoCombat
    rule=ItemFinalPoseOverride.AnimGraphNode_StateMachine_0.StateMachine.AnimStateTransitionNode_10.Transition
    src=AlertedCycle
    via=AnimStateTransitionNode_10
   }
   {
    dst=Unaware
    rule=ItemFinalPoseOverride.AnimGraphNode_StateMachine_0.StateMachine.AnimStateTransitionNode_13.Transition
    src=AlertedCycle
    via=AnimStateTransitionNode_13
   }
   {
    dst=Death
    rule=ItemFinalPoseOverride.AnimGraphNode_StateMachine_0.StateMachine.AnimStateTransitionNode_0.Transition
    src=Alias
    via=AnimStateTransitionNode_0
   }
   {
    dst=CombatToAlerted
    rule=ItemFinalPoseOverride.AnimGraphNode_StateMachine_0.StateMachine.AnimStateTransitionNode_14.Transition
    src=Combat
    via=AnimStateTransitionNode_14
   }
   {
    dst=AlertedCycle
    rule=ItemFinalPoseOverride.AnimGraphNode_StateMachine_0.StateMachine.AnimStateTransitionNode_15.Transition
    src=CombatToAlerted
    via=AnimStateTransitionNode_15
   }
   {
    dst=TransIntoCombat
    rule=ItemFinalPoseOverride.AnimGraphNode_StateMachine_0.StateMachine.AnimStateTransitionNode_21.Transition
    src=ToCombat
    via=AnimStateTransitionNode_21
   }
   {
    dst=Combat
    rule=ItemFinalPoseOverride.AnimGraphNode_StateMachine_0.StateMachine.AnimStateTransitionNode_11.Transition
    src=TransIntoCombat
    via=AnimStateTransitionNode_11
   }
   {
    dst=Combat
    rule=ItemFinalPoseOverride.AnimGraphNode_StateMachine_0.StateMachine.AnimStateTransitionNode_22.Transition
    src=TransIntoCombat
    via=AnimStateTransitionNode_22
   }
   {
    dst=intoAlert
    rule=ItemFinalPoseOverride.AnimGraphNode_StateMachine_0.StateMachine.AnimStateTransitionNode_1.Transition
    src=Unaware
    via=AnimStateTransitionNode_1
   }
   {
    dst=AlertedCycle
    rule=ItemFinalPoseOverride.AnimGraphNode_StateMachine_0.StateMachine.AnimStateTransitionNode_2.Transition
    src=intoAlert
    via=AnimStateTransitionNode_2
   }
  ]
 }
 Death {
  path=ItemFinalPoseOverride.AnimGraphNode_StateMachine_0.StateMachine.AnimStateNode_0.Death
  nodes [Output Animation Pose(Result=Use cached pose 'FinalPoseCache'())]
 }
 TransIntoCombat {
  path=ItemFinalPoseOverride.AnimGraphNode_StateMachine_0.StateMachine.AnimStateNode_10.TransIntoCombat
  nodes [
   n1 = Sequence Player 'FruitFire_NudgeHonk_FloatToSummon'() {
    bLoopAnimation=false
    groupName=Land
    groupRole=AlwaysFollower
    method=SyncGroup
    sequence=&/FruitFire_NudgeHonk_FloatToSummon
   }
   n2 = Use cached pose 'FinalPoseCache'()
   n3 = Sequence Player 'FruitFire_NudgeHonk_FloatToSummon'() {
    bLoopAnimation=false
    groupName=Land
    groupRole=AlwaysLeader
    method=SyncGroup
    sequence=&FruitFire_NudgeHonk_FloatToSummon
   }
   Output Animation Pose(Result=Blend Poses by bool(
     BlendPose_0=Layered blend per bone(BasePose=@n2, BlendPoses_0=@n1) {layerSetup [{branchFilters [{boneName=Spine_01}]}]}
     BlendPose_1=@n3
     BlendTime_0=0.2
     bActiveValue=Get bExitLand() -> (bExitLand:bool)
    ))
  ]
 }
 CombatToAlerted {
  path=ItemFinalPoseOverride.AnimGraphNode_StateMachine_0.StateMachine.AnimStateNode_11.CombatToAlerted
  nodes [Output Animation Pose(Result=Sequence Player 'FruitFire_NudgeHonk_CombatToAlert'() {bLoopAnimation=false, sequence=&/FruitFire_NudgeHonk_CombatToAlert})]
 }
 Combat {
  path=ItemFinalPoseOverride.AnimGraphNode_StateMachine_0.StateMachine.AnimStateNode_5.Combat
  nodes [Output Animation Pose(Result=Use cached pose 'FinalPoseCache'())]
 }
 Unaware {
  path=ItemFinalPoseOverride.AnimGraphNode_StateMachine_0.StateMachine.AnimStateNode_6.Unaware
  nodes [
   Sequence Player 'FruitFire_NudgeHonk_RideCycle'() {sequence=&/FruitFire_NudgeHonk_RideCycle}
   Output Animation Pose(Result=Use cached pose 'FinalPoseCache'())
  ]
 }
 intoAlert {
  path=ItemFinalPoseOverride.AnimGraphNode_StateMachine_0.StateMachine.AnimStateNode_7.intoAlert
  nodes [Output Animation Pose(Result=Sequence Player 'FruitFire_NudgeHonk_RideToFloat'() {bLoopAnimation=false, sequence=&/FruitFire_NudgeHonk_RideToFloat})]
 }
 AlertedCycle {
  path=ItemFinalPoseOverride.AnimGraphNode_StateMachine_0.StateMachine.AnimStateNode_8.AlertedCycle
  nodes [Output Animation Pose(Result=Sequence Player 'FruitFire_NudgeHonk_FloatCycle'() {bLoopAnimation=false, sequence=&/FruitFire_NudgeHonk_FloatCycle})]
 }
 Transition {
  path=ItemFinalPoseOverride.AnimGraphNode_StateMachine_0.StateMachine.AnimStateTransitionNode_0.Transition
  nodes [Result(bCanEnterTransition=Get bDead() -> (bDead:bool))]
 }
 Transition {
  path=ItemFinalPoseOverride.AnimGraphNode_StateMachine_0.StateMachine.AnimStateTransitionNode_1.Transition
  nodes [Result(bCanEnterTransition=Get bAlerted() -> (bAlerted:bool))]
 }
 Transition {
  path=ItemFinalPoseOverride.AnimGraphNode_StateMachine_0.StateMachine.AnimStateTransitionNode_10.Transition
  nodes [Result(bCanEnterTransition=Get bInCombat() -> (bInCombat:bool))]
 }
 Transition {path=ItemFinalPoseOverride.AnimGraphNode_StateMachine_0.StateMachine.AnimStateTransitionNode_11.Transition, nodes [Result()]}
 Transition {
  path=ItemFinalPoseOverride.AnimGraphNode_StateMachine_0.StateMachine.AnimStateTransitionNode_13.Transition
  nodes [Result(bCanEnterTransition=NOT Boolean(A=Get bAlerted()))]
 }
 Transition {
  path=ItemFinalPoseOverride.AnimGraphNode_StateMachine_0.StateMachine.AnimStateTransitionNode_14.Transition
  nodes [Result(bCanEnterTransition=NOT Boolean(A=Get bInCombat()))]
 }
 Transition {path=ItemFinalPoseOverride.AnimGraphNode_StateMachine_0.StateMachine.AnimStateTransitionNode_15.Transition, nodes [Result()]}
 Transition {path=ItemFinalPoseOverride.AnimGraphNode_StateMachine_0.StateMachine.AnimStateTransitionNode_2.Transition, nodes [Result()]}
 Transition {
  path=ItemFinalPoseOverride.AnimGraphNode_StateMachine_0.StateMachine.AnimStateTransitionNode_21.Transition
  nodes [Result(bCanEnterTransition=Get bInCombat())]
 }
 Transition {
  path=ItemFinalPoseOverride.AnimGraphNode_StateMachine_0.StateMachine.AnimStateTransitionNode_22.Transition
  nodes [Result(bCanEnterTransition=Get bIsFiring() -> (bIsFiring:bool))]
 }
 ItemFullBodyAimOffsets {nodes [
   Input Pose!() -> (Pose, FullBody_AimOffsetAlpha:double, Aim_Yaw:double, Aim_Pitch:double) {name=InPoseFullBodyAimOffsets}
   Input Pose!() -> (Pose) {name=InPosePreFullBodyAimOffsets}
   Output Pose() {layerGroup=Item Override}
  ]}
 ItemFullBodyOverride {nodes [
   Input Pose() {name=InPoseFullBody}
   Output Pose() {layerGroup=Item Override}
  ]}
 ItemIdleOverride {nodes [
   Input Pose() {name=InPose_Idle}
   Output Pose() {layerGroup=Item Override}
  ]}
 ItemJetPackJumpAdditive {nodes [Output Pose() {layerGroup=Item Override}]}
 ItemJetPackStartAdditive {nodes [Output Pose() {layerGroup=Item Override}]}
 ItemJumpLoopAdditive {nodes [
   Input Pose() {name=InPoseItemJumpLoopAdditive}
   Output Pose() {layerGroup=Item Override}
  ]}
 ItemJumpOverride {nodes [
   Input Pose() {name=InPoseJump}
   Output Pose() {layerGroup=Item Override}
  ]}
 ItemJumpUpAdditive {nodes [
   Input Pose() {name=InPoseItemJumpUpAdditive}
   Output Pose() {layerGroup=Item Override}
  ]}
 ItemPreAimOffsets {nodes [
   Input Pose() {name=InPosePreAimOffsets}
   Output Pose() {layerGroup=Item Override}
  ]}
 ItemPreIK {nodes [
   Input Pose() {name=InPose_PreIK}
   Output Pose() {layerGroup=Item Override}
  ]}
 ItemRightHandIK {nodes [
   Input Pose() {name=InPosePreRightHandIK}
   Input Pose() {name=InPosePostRightHandIK}
   Output Pose() {layerGroup=Item Override}
  ]}
 ItemSkeletalControl {nodes [
   Input Pose()
   Output Pose() {layerGroup=Item Override}
  ]}
 ItemSwimJumpFallLoop {nodes [Output Pose() {layerGroup=Item Override}]}
 ItemSwimJumpSurfaceEnd {nodes [Output Pose() {layerGroup=Item Override}]}
 ItemSwimJumpSurfaceLoop {nodes [Output Pose() {layerGroup=Item Override}]}
 ItemUpperBody {nodes [
   Input Pose!() -> (Pose, InputParam:Fort Anim Input Adjusted Aim Structure) {name=InPoseUpperBody}
   Output Pose() {layerGroup=Item Override}
  ]}
 ItemZipLineStartAdditive {nodes [Output Pose() {layerGroup=Item Override}]}
]
variables []
